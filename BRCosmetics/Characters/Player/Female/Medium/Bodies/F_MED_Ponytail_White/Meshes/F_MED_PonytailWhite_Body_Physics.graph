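# Physics asset for the medium female Ponytail White cosmetic body, apparently for secondary motion of its leg bags, leg straps and shoe lips.
# Capsules cover pelvis, thighs and feet, with small ones on spine and thigh twists, linked by free-swing, free-twist constraints.
# Box bodies on the dyn_ bones attach to thighs, thigh twists and feet with angular limits, tighter on bags and shoe lips than on straps.
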PhysicsAsset /BRCosmetics/Characters/Player/Female/Medium/Bodies/F_MED_Ponytail_White/Meshes/F_MED_PonytailWhite_Body_Physics
bodies {
 dyn_l_legBag [{shapeType=Box, extentX=22.2175, extentY=3.48, extentZ=19.6234, center=<0.7124 0.1977 -1.5611>, rotation=<rot 0 89.6703 -88.8054>}]
 dyn_l_legStrap [{shapeType=Box, extentX=6.3608, extentY=8.1517, extentZ=0.799, center=<-0.0001 -1.801 -0.9416>, rotation=<rot -13.4068 7.1558 -28.4309>}]
 dyn_l_shoeLip [{shapeType=Box, extentX=1.362, extentY=5.1722, extentZ=5.503, center=<0.1748 2.5673 0>, rotation=<rot 0 -1.4526 0>}]
 dyn_r_legBag [{shapeType=Box, extentX=22.2175, extentY=2.9796, extentZ=19.6234, center=<0.6553 0.1983 1.1773>, rotation=<rot 0 89.6703 -88.8054>}]
 dyn_r_legStrap [{shapeType=Box, extentX=6.3608, extentY=8.1517, extentZ=0.799, center=<0 1.3312 0.6957>, rotation=<rot -13.4069 7.1553 -28.4312>}]
 dyn_r_shoeLip [{shapeType=Box, extentX=1.362, extentY=5.1722, extentZ=5.503, center=<0.1748 2.5673 0>, rotation=<rot 0 -1.4526 0>}]
 foot_l [{shapeType=Capsule, length=12.1358, radius=5.2745, center=<-2.3926 6.2461 0>, rotation=<rot 90 -14.0363 55.0042>}]
 foot_r [{shapeType=Capsule, length=12.1358, radius=5.2745, center=<2.3076 -6.0237 0>, rotation=<rot 90 -18.435 50.6048>}]
 pelvis [{shapeType=Capsule, length=9.8293, radius=10.2097}]
 spine_01 [{shapeType=Capsule, length=2, radius=2}]
 spine_02 [{shapeType=Capsule, length=2, radius=2}]
 spine_03 [{shapeType=Capsule, length=2, radius=2}]
 thigh_l [{shapeType=Capsule, length=24.2399, radius=7.3469, center=<-19.0041 0.0005 0>, rotation=<rot 90 0 0>}]
 thigh_r [{shapeType=Capsule, length=24.2399, radius=7.3469, center=<19.2903 0.0002 0>, rotation=<rot 90 0 0>}]
 thigh_twist_01_l [{shapeType=Capsule, length=2, radius=2}]
 thigh_twist_01_r [{shapeType=Capsule, length=2, radius=2}]
}
constraints [
 {
  bone1Name=spine_01
  bone2Name=pelvis
  swing1LimitDegrees=45
  swing1Motion=Free
  swing2LimitDegrees=45
  swing2Motion=Free
  twistLimitDegrees=45
  twistMotion=Free
 }
 {
  bone1Name=spine_02
  bone2Name=spine_01
  swing1LimitDegrees=45
  swing1Motion=Free
  swing2LimitDegrees=45
  swing2Motion=Free
  twistLimitDegrees=45
  twistMotion=Free
 }
 {
  bone1Name=spine_03
  bone2Name=spine_02
  swing1LimitDegrees=45
  swing1Motion=Free
  swing2LimitDegrees=45
  swing2Motion=Free
  twistLimitDegrees=45
  twistMotion=Free
 }
 {
  bone1Name=thigh_l
  bone2Name=pelvis
  swing1LimitDegrees=45
  swing1Motion=Free
  swing2LimitDegrees=45
  swing2Motion=Free
  twistLimitDegrees=45
  twistMotion=Free
 }
 {
  bone1Name=thigh_r
  bone2Name=pelvis
  swing1LimitDegrees=45
  swing1Motion=Free
  swing2LimitDegrees=45
  swing2Motion=Free
  twistLimitDegrees=45
  twistMotion=Free
 }
 {bone1Name=dyn_l_legBag, bone2Name=thigh_l, swing1LimitDegrees=15, swing2LimitDegrees=15, twistLimitDegrees=15}
 {bone1Name=dyn_r_legBag, bone2Name=thigh_r, swing1LimitDegrees=15, swing2LimitDegrees=15, twistLimitDegrees=15}
 {bone1Name=dyn_l_shoeLip, bone2Name=foot_l, swing1LimitDegrees=30, swing2LimitDegrees=15, twistLimitDegrees=15}
 {bone1Name=dyn_r_shoeLip, bone2Name=foot_r, swing1LimitDegrees=30, swing2LimitDegrees=15, twistLimitDegrees=15}
 {
  bone1Name=foot_r
  bone2Name=thigh_r
  swing1LimitDegrees=45
  swing1Motion=Free
  swing2LimitDegrees=45
  swing2Motion=Free
  twistLimitDegrees=45
  twistMotion=Free
 }
 {
  bone1Name=foot_l
  bone2Name=thigh_l
  swing1LimitDegrees=45
  swing1Motion=Free
  swing2LimitDegrees=45
  swing2Motion=Free
  twistLimitDegrees=45
  twistMotion=Free
 }
 {bone1Name=dyn_l_legStrap, bone2Name=thigh_twist_01_l, swing1LimitDegrees=50, swing2LimitDegrees=50, twistLimitDegrees=50}
 {bone1Name=dyn_r_legStrap, bone2Name=thigh_twist_01_r, swing1LimitDegrees=50, swing2LimitDegrees=50, twistLimitDegrees=50}
 {
  bone1Name=thigh_twist_01_l
  bone2Name=thigh_l
  swing1LimitDegrees=45
  swing1Motion=Free
  swing2LimitDegrees=45
  swing2Motion=Free
  twistLimitDegrees=45
  twistMotion=Free
 }
 {
  bone1Name=thigh_twist_01_r
  bone2Name=thigh_r
  swing1LimitDegrees=45
  swing1Motion=Free
  swing2LimitDegrees=45
  swing2Motion=Free
  twistLimitDegrees=45
  twistMotion=Free
 }
]
