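# Post-processing-style AnimBlueprint for the White Ponytail medium female body, layering secondary motion and corrective bone fixes onto the pose copied from its attached parent mesh.
# Shoe-lip bones are constrained to the calves, then a low-alpha RigidBody node simulates the body physics asset relative to root.
# Additive bone tweaks offset the left butt bone while crouching and scale the knee-fix bone while jumping or falling, using binary alphas.
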AnimBlueprint /BRCosmetics/Characters/Player/Female/Medium/Bodies/F_MED_Ponytail_White/Meshes/F_MED_PonytailWhite_AnimBP
target_skeleton=&/F_MED_PonytailWhite_Skeleton
graphs [
 AnimGraph {nodes [
   n1 = "Constraint - dyn_l_shoeLip (P)"(
    ComponentPose=Local To Component(LocalPose=Copy Pose From Mesh() {bCopyCurves=true, bUseAttachedParent=true})
    Alpha=1
    ConstraintWeights_0=1
   ) {
    alphaBoolBlend {blendOption=Linear}
    boneToModify {boneName=dyn_l_shoeLip}
    constraintSetup [{targetBone {boneName=calf_l}}]
    constraintWeights [1]
   }
   n2 = "Constraint - dyn_r_shoeLip (P)"(ComponentPose=@n1, Alpha=1, ConstraintWeights_0=1) {
    alphaBoolBlend {blendOption=Linear}
    boneToModify {boneName=dyn_r_shoeLip}
    constraintSetup [{targetBone {boneName=calf_r}}]
    constraintWeights [1]
   }
   n3 = RigidBody(ComponentPose=@n2, Alpha=0.3) {
    bClampLinearTranslationLimitToRefPose=true
    bOverrideWorldGravity=true
    overridePhysicsAsset=&/F_MED_PonytailWhite_Body_Physics
    simulationSpace=BaseBoneSpace
    alphaBoolBlend {blendOption=Linear}
    baseBoneRef {boneName=root}
    componentAppliedLinearAccClamp=<2500 2500 2500>
    componentLinearAccScale=<0.25 0.25 0.45>
    overrideWorldGravity {z=-980}
   }
   n4 = "Transform (Modify) Bone - Bone: dyn_l_butt"(ComponentPose=@n3, Alpha=Get Crouch_Alpha() -> (Crouch_Alpha:double), Translation="0,-2,-12", Rotation="0,0,0", Scale="1,1,1") {
    rotationMode=BMM_Additive
    rotationSpace=BCS_WorldSpace
    scaleMode=BMM_Additive
    scaleSpace=BCS_WorldSpace
    translationMode=BMM_Additive
    translationSpace=BCS_WorldSpace
    alphaBoolBlend {blendOption=Linear}
    boneToModify {boneName=dyn_l_butt}
    translation {y=-2, z=-12}
   }
   n5 = Get bIsJumping()
   n6 = Get bIsFalling()
   n7 = "Transform (Modify) Bone - Bone: dyn_kneeFix"(ComponentPose=@n4, Alpha="To Float (Boolean)"(InBool=OR Boolean(A=@n5, B=@n6)), Translation="0,0,0", Rotation="0,0,0", Scale="1,3,1") {
    rotationMode=BMM_Additive
    scaleMode=BMM_Additive
    translationMode=BMM_Additive
    alphaBoolBlend {blendOption=Linear}
    boneToModify {boneName=dyn_kneeFix}
    scale {y=3}
   }
   Output Pose(Result=Component To Local(ComponentPose=@n7)) {layerGroup=None}
  ]}
 EventGraph {nodes [
   Event BlueprintUpdateAnimation()
   > Set Crouch_Alpha(Crouch_Alpha:double=SelectFloat(A=1, B=0, bPickA=Get bIsCrouching())) -> (Output_Get:double)
   TryGetPawnOwner()
  ]}
]
variables [Crouch_Alpha {type=number}]
