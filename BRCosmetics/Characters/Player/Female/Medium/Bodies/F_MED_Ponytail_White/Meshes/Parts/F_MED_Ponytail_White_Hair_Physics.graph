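# Physics asset for the hair of the female medium white-ponytail player character, giving its dynamic hair bones bodies for simulation.
# Nine hair chains (a–i) of two to four thin box bodies each are constrained bone to bone from the head, with uniform swing and twist limits.
# Capsules shape the head and both neck bones, which are linked by constraints with free swing and twist.
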PhysicsAsset /BRCosmetics/Characters/Player/Female/Medium/Bodies/F_MED_Ponytail_White/Meshes/Parts/F_MED_Ponytail_White_Hair_Physics
bodies {
 dyn_hair_a_01 [{shapeType=Box, extentX=3.3333, extentY=3.4669, extentZ=1.2063, center=<1.6548 0 0.0001>}]
 dyn_hair_a_02 [{shapeType=Box, extentX=3.3333, extentY=3.4669, extentZ=1.2063, center=<1.6548 0 0.0001>}]
 dyn_hair_b_01 [{shapeType=Box, extentX=4.5251, extentY=4.3611, extentZ=1.5358, center=<2.3999 0 0>}]
 dyn_hair_b_02 [{shapeType=Box, extentX=5.8678, extentY=4.3611, extentZ=1.5358, center=<2.8978 0 -0.0001>}]
 dyn_hair_b_03 [{shapeType=Box, extentX=5.9404, extentY=2.8202, extentZ=1.5358, center=<2.4916 0 0>}]
 dyn_hair_c_01 [{shapeType=Box, extentX=4.5251, extentY=4.3611, extentZ=1.5358, center=<2.3999 0 0>}]
 dyn_hair_c_02 [{shapeType=Box, extentX=5.8678, extentY=4.3611, extentZ=1.1756, center=<2.8978 0 0.0691>}]
 dyn_hair_c_03 [{shapeType=Box, extentX=4.5251, extentY=4.3611, extentZ=1.5358, center=<2.3999 0 0>}]
 dyn_hair_c_04 [{shapeType=Box, extentX=4.5251, extentY=3.6334, extentZ=1.5358, center=<2.3999 0 0>}]
 dyn_hair_d_01 [{shapeType=Box, extentX=4.5251, extentY=4.3611, extentZ=0.8573, center=<2.3999 0 0.1743>}]
 dyn_hair_d_02 [{shapeType=Box, extentX=6.3525, extentY=4.3611, extentZ=1.1154, center=<3.2004 0 0.1676>}]
 dyn_hair_d_03 [{shapeType=Box, extentX=5.4881, extentY=4.6607, extentZ=1.5358, center=<2.5367 0 0>}]
 dyn_hair_e_01 [{shapeType=Box, extentX=5.0454, extentY=4.3611, extentZ=1.093, center=<2.5005 0 0>}]
 dyn_hair_e_02 [{shapeType=Box, extentX=5.9919, extentY=4.3611, extentZ=1.1375, center=<3.0933 0 0.1625>}]
 dyn_hair_e_03 [{shapeType=Box, extentX=3.4602, extentY=4.3611, extentZ=1.5358, center=<1.6984 0 0>}]
 dyn_hair_f_01 [{shapeType=Box, extentX=6.1377, extentY=4.3611, extentZ=1.5358, center=<3.2201 0 0.0002>}]
 dyn_hair_f_02 [{shapeType=Box, extentX=5.0454, extentY=4.3611, extentZ=1.5358, center=<2.5005 0 0>}]
 dyn_hair_f_03 [{shapeType=Box, extentX=3.0408, extentY=4.3611, extentZ=1.5358, center=<1.5377 0 0>}]
 dyn_hair_g_01 [{shapeType=Box, extentX=6.1377, extentY=4.3611, extentZ=1.5358, center=<3.2201 0 0.0002>}]
 dyn_hair_g_02 [{shapeType=Box, extentX=4.8601, extentY=4.3611, extentZ=1.5358, center=<2.4269 0 0.0002>}]
 dyn_hair_g_03 [{shapeType=Box, extentX=4.4409, extentY=4.3611, extentZ=1.5358, center=<2.1533 0 -0.3971>, rotation=<rot -13.0421 0 -0.0001>}]
 dyn_hair_h_01 [{shapeType=Box, extentX=5.7395, extentY=4.3611, extentZ=1.5358, center=<3.0212 0 0.0002>}]
 dyn_hair_h_02 [{shapeType=Box, extentX=5.1738, extentY=4.3611, extentZ=1.5358, center=<2.5728 0 0.0002>}]
 dyn_hair_h_03 [{shapeType=Box, extentX=3.1785, extentY=4.3611, extentZ=1.5358, center=<1.5586 -0.1765 0>, rotation=<rot 1.8863 -1.939 0>}]
 dyn_hair_h_04 [{shapeType=Box, extentX=3.1785, extentY=4.3611, extentZ=1.5358, center=<1.5586 0 0>}]
 dyn_hair_i_01 [{shapeType=Box, extentX=6.171, extentY=4.3611, extentZ=1.5358, center=<3.3602 0 0.0001>}]
 dyn_hair_i_02 [{shapeType=Box, extentX=6.1339, extentY=4.5776, extentZ=1.5358, center=<3.0723 0 0.0001>}]
 dyn_hair_i_03 [{shapeType=Box, extentX=3.1015, extentY=4.1846, extentZ=1.5358, center=<1.5424 0 0>}]
 dyn_hair_i_04 [{shapeType=Box, extentX=3.1015, extentY=2.2651, extentZ=1.5358, center=<1.5424 0 0>}]
 head [
  {shapeType=Capsule, length=4.1536, radius=6.4873, center=<2.8136 -0.4449 -0.0001>, rotation=<rot 90 24.6235 33.608>}
  {shapeType=Capsule, length=8.5768, radius=5.961, center=<-0.4088 3.4682 0.0003>, rotation=<rot 90 0 0>}
 ]
 neck_01 [{shapeType=Capsule, length=2.5859, radius=3}]
 neck_02 [{shapeType=Capsule, length=2.5859, radius=3}]
}
constraints [
 {bone1Name=dyn_hair_i_01, bone2Name=head, swing1LimitDegrees=45, swing2LimitDegrees=45, twistLimitDegrees=45}
 {bone1Name=dyn_hair_i_02, bone2Name=dyn_hair_i_01, swing1LimitDegrees=45, swing2LimitDegrees=45, twistLimitDegrees=45}
 {bone1Name=dyn_hair_i_03, bone2Name=dyn_hair_i_02, swing1LimitDegrees=45, swing2LimitDegrees=45, twistLimitDegrees=45}
 {bone1Name=dyn_hair_i_04, bone2Name=dyn_hair_i_03, swing1LimitDegrees=45, swing2LimitDegrees=45, twistLimitDegrees=45}
 {bone1Name=dyn_hair_a_01, bone2Name=head, swing1LimitDegrees=45, swing2LimitDegrees=45, twistLimitDegrees=45}
 {bone1Name=dyn_hair_a_02, bone2Name=dyn_hair_a_01, swing1LimitDegrees=45, swing2LimitDegrees=45, twistLimitDegrees=45}
 {bone1Name=dyn_hair_b_01, bone2Name=head, swing1LimitDegrees=45, swing2LimitDegrees=45, twistLimitDegrees=45}
 {bone1Name=dyn_hair_b_02, bone2Name=dyn_hair_b_01, swing1LimitDegrees=45, swing2LimitDegrees=45, twistLimitDegrees=45}
 {bone1Name=dyn_hair_b_03, bone2Name=dyn_hair_b_02, swing1LimitDegrees=45, swing2LimitDegrees=45, twistLimitDegrees=45}
 {bone1Name=dyn_hair_c_01, bone2Name=head, swing1LimitDegrees=45, swing2LimitDegrees=45, twistLimitDegrees=45}
 {bone1Name=dyn_hair_c_02, bone2Name=dyn_hair_c_01, swing1LimitDegrees=45, swing2LimitDegrees=45, twistLimitDegrees=45}
 {bone1Name=dyn_hair_c_03, bone2Name=dyn_hair_c_02, swing1LimitDegrees=45, swing2LimitDegrees=45, twistLimitDegrees=45}
 {bone1Name=dyn_hair_c_04, bone2Name=dyn_hair_c_03, swing1LimitDegrees=45, swing2LimitDegrees=45, twistLimitDegrees=45}
 {bone1Name=dyn_hair_d_01, bone2Name=head, swing1LimitDegrees=45, swing2LimitDegrees=45, twistLimitDegrees=45}
 {bone1Name=dyn_hair_d_02, bone2Name=dyn_hair_d_01, swing1LimitDegrees=45, swing2LimitDegrees=45, twistLimitDegrees=45}
 {bone1Name=dyn_hair_d_03, bone2Name=dyn_hair_d_02, swing1LimitDegrees=45, swing2LimitDegrees=45, twistLimitDegrees=45}
 {bone1Name=dyn_hair_e_01, bone2Name=head, swing1LimitDegrees=45, swing2LimitDegrees=45, twistLimitDegrees=45}
 {bone1Name=dyn_hair_e_02, bone2Name=dyn_hair_e_01, swing1LimitDegrees=45, swing2LimitDegrees=45, twistLimitDegrees=45}
 {bone1Name=dyn_hair_e_03, bone2Name=dyn_hair_e_02, swing1LimitDegrees=45, swing2LimitDegrees=45, twistLimitDegrees=45}
 {bone1Name=dyn_hair_f_01, bone2Name=head, swing1LimitDegrees=45, swing2LimitDegrees=45, twistLimitDegrees=45}
 {bone1Name=dyn_hair_f_02, bone2Name=dyn_hair_f_01, swing1LimitDegrees=45, swing2LimitDegrees=45, twistLimitDegrees=45}
 {bone1Name=dyn_hair_f_03, bone2Name=dyn_hair_f_02, swing1LimitDegrees=45, swing2LimitDegrees=45, twistLimitDegrees=45}
 {bone1Name=dyn_hair_g_01, bone2Name=head, swing1LimitDegrees=45, swing2LimitDegrees=45, twistLimitDegrees=45}
 {bone1Name=dyn_hair_g_02, bone2Name=dyn_hair_g_01, swing1LimitDegrees=45, swing2LimitDegrees=45, twistLimitDegrees=45}
 {bone1Name=dyn_hair_g_03, bone2Name=dyn_hair_g_02, swing1LimitDegrees=45, swing2LimitDegrees=45, twistLimitDegrees=45}
 {bone1Name=dyn_hair_h_01, bone2Name=head, swing1LimitDegrees=45, swing2LimitDegrees=45, twistLimitDegrees=45}
 {bone1Name=dyn_hair_h_02, bone2Name=dyn_hair_h_01, swing1LimitDegrees=45, swing2LimitDegrees=45, twistLimitDegrees=45}
 {bone1Name=dyn_hair_h_03, bone2Name=dyn_hair_h_02, swing1LimitDegrees=45, swing2LimitDegrees=45, twistLimitDegrees=45}
 {bone1Name=dyn_hair_h_04, bone2Name=dyn_hair_h_03, swing1LimitDegrees=45, swing2LimitDegrees=45, twistLimitDegrees=45}
 {
  bone1Name=head
  bone2Name=neck_02
  swing1LimitDegrees=45
  swing1Motion=Free
  swing2LimitDegrees=45
  swing2Motion=Free
  twistLimitDegrees=45
  twistMotion=Free
 }
 {
  bone1Name=neck_02
  bone2Name=neck_01
  swing1LimitDegrees=45
  swing1Motion=Free
  swing2LimitDegrees=45
  swing2Motion=Free
  twistLimitDegrees=45
  twistMotion=Free
 }
]
