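# Animation Blueprint for the white ponytail part of the medium female player body, adding secondary physics motion to the hair.
# It copies the pose from the attached parent mesh, then runs a RigidBody node with a hair physics asset, simulated relative to neck_01 with overridden gravity.
# Physics alpha drops from full to a low blend while skydiving or in dive mode.
# OnEndSkydiving sets two variables that the AnimGraph does not read.
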AnimBlueprint /BRCosmetics/Characters/Player/Female/Medium/Bodies/F_MED_Ponytail_White/Meshes/Parts/F_MED_Ponytail_White_AnimBP
target_skeleton=&/F_MED_Ponytail_White_Skeleton
graphs [
 AnimGraph {nodes [
   n1 = Get bIsSkyDiving()
   n2 = Get bIsSkydiveDiveMode()
   n3 = RigidBody(
    ComponentPose=Local To Component(LocalPose=Copy Pose From Mesh() {bUseAttachedParent=true})
    Alpha=SelectFloat(A=0.2, B=1, bPickA=OR Boolean(A=@n1, B=@n2))
   ) {
    bClampLinearTranslationLimitToRefPose=true
    bOverrideWorldGravity=true
    overridePhysicsAsset=&/F_MED_Ponytail_White_Hair_Physics
    simulationSpace=BaseBoneSpace
    alphaBoolBlend {blendOption=Linear}
    baseBoneRef {boneName=neck_01}
    componentAppliedLinearAccClamp=<700 700 700>
    componentLinearAccScale {y=0.3, z=0.3}
    overrideWorldGravity {z=-980}
   }
   Output Pose(Result=Component To Local(ComponentPose=@n3)) {layerGroup=None}
  ]}
 EventGraph {nodes [
   Event OnEndSkydiving()
   > Set isSkydiving(isSkydiving:bool=Get bIsSkyDiving()) -> (Output_Get:bool)
   > Set isSkydivingINV(isSkydivingINV:double=0) -> (Output_Get:double)
   Event BlueprintUpdateAnimation()
  ]}
]
variables [
 isSkydiving {type=boolean}
 isSkydivingINV {type=number}
]
